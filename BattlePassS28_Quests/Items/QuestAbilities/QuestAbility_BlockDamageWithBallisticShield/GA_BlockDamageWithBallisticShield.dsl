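; Gameplay ability for a Battle Pass Season 28 quest about blocking damage with the Ballistic Shield.
; The activate and end-ability events are present but empty, so the graph itself holds no logic.
; Any quest tracking behaviour would come from the parent class or defaults, which this graph does not show.
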(event Ability|EventActivateAbility)

(event Ability|EventOnEndAbility
  (params (bWasCancelled :type Boolean)))
